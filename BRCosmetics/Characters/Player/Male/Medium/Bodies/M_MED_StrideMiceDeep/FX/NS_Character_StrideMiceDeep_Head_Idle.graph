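# Audio-reactive idle head FX for the StrideMiceDeep character cosmetic, with colors chosen by User.Variant.
# Flares spawns short-lived sprites at preset skeletal-socket offsets, with rate scaled by System.AudioReactivity, which a fake audio loop module can simulate.
# Headwaves bursts paired socket-attached sprites, feeding positions and reactive values to dynamic material parameters.
# Front-end, quality and time-of-day inputs tune brightness, and bisHeadHidden scales both emitters to zero.
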NiagaraSystem /BRCosmetics/Characters/Player/Male/Medium/Bodies/M_MED_StrideMiceDeep/FX/NS_Character_StrideMiceDeep_Head_Idle
dependencies {
 usedDataInterfaces [
  &NiagaraDataInterfaceArrayFloat
  &NiagaraDataInterfaceArrayInt32
  &NiagaraDataInterfaceArrayFloat3
  &NiagaraDataInterfaceCurve
  &NiagaraDataInterfaceSkeletalMesh
  &NiagaraDataInterfaceColorCurve
 ]
 usedDynamicInputs [
  &ColorFromCurve
  /Niagara/DynamicInputs/Lerp/SmoothLerpOverTime_Float
  &RemapRange
 ]
 usedModules [
  &self:SystemSpawnScript.SystemScriptSource.SystemScriptGraph.NiagaraNodeAssignment_3.SetVariables_#1
  &self:SystemSpawnScript.SystemScriptSource.SystemScriptGraph.NiagaraNodeAssignment_4.SetVariables_#2
  &self:SystemSpawnScript.SystemScriptSource.SystemScriptGraph.NiagaraNodeAssignment_2.SetVariables_#3
  &SystemState
  /Game/Effects/Niagara/Modules/Audio/FN_FakeAudioReactivityLoop
  &self:SystemSpawnScript.SystemScriptSource.SystemScriptGraph.NiagaraNodeAssignment_1.SetVariables_#4
  &self:SystemSpawnScript.SystemScriptSource.SystemScriptGraph.NiagaraNodeAssignment_0.SetVariables_#5
  &self:SystemSpawnScript.SystemScriptSource.SystemScriptGraph.NiagaraNodeAssignment_6.SetVariables_#6
  &Timeline
 ]
}
emitters {
 Flares {
  emitterSpawnScript [/SetVariables(Emitter.FlareColor:LinearColor=&MergeHelper_ColorSwitch(
     0=<rgba 0.2958 0.1198 1 5>
     1=<rgba 0.6354 0.3581 0.1601 3>
     2=<rgba 0.2327 0.4273 0.8594 1.5>
     3=?
     Index=&MakeIntFromFloat(Float=$User.Variant)
     Num=::NewEnumerator1
    )) {isSetParameters=true}]
  emitterUpdateScript [
   &EmitterState(
    Inactive Response=::NewEnumerator1
    Life Cycle Mode=::NewEnumerator1
    Loop Duration=0.35
    Recalculate Duration Each Loop=true
    UseLoopDelay=false
   )
   &SpawnRate(
    Spawn Probability=&MakeCustomFloatFromBool(Bool=&SetBoolByFloatComparison(A=$System.AudioReactivity, B=0.1), True Float=0.8, False Float=0)
    SpawnRate=&Multiply_Float(
     A=&RandomRangeFloat(Minimum=4, Maximum=8)
     B=&Lerp_Float(A=0, B=1, Alpha=$System.AudioReactivity)
    )
    Use Spawn Probability=true
   )
  ]
  particleSpawnScript [
   &PartitionParticles(Number Of Partitions=19, Partition Mode:ENiagara_PartitionMode_v2=::NewEnumerator5)
   &InitializeParticle(
    Color Mode=::NewEnumerator0
    Lifetime=&RandomRangeFloat(Minimum=0.3, Maximum=0.5)
    Sprite Rotation Angle=25
    Sprite Rotation Mode=::NewEnumerator0
    Sprite Size Mode=::NewEnumerator3
    Uniform Sprite Size=&FloatFromCurve(
     FloatCurve=DataInterface() {propertyValues {Curve {keys [<at 0 20>,<at 0.85 35>,<at 1 50>]}}}
     CurveIndex=&RandomRangeFloat(Minimum=0, Maximum=1)
    )
    UsePositionOffset=false
    Write Material Random=true
   )
   &CameraOffset(Camera Offset Amount=&RandomRangeFloat(Minimum=3, Maximum=5))
  ]
  particleUpdateScript [
   &ParticleState()
   &ScaleSpriteSize(
    Scale Sprite Size Mode=::NewEnumerator0
    Uniform Scale Factor=&Multiply_Float(
     A=&FloatFromCurve(FloatCurve=DataInterface() {propertyValues {Curve {keys [<at 0 0>,<at 0.5 1>,<at 1 0>]}}})
     B=&MakeCustomFloatFromBool(Bool=$User.bisHeadHidden, True Float=0, False Float=1)
    )
   )
   &SkeletalMeshLocation(
    MaskSpawnGroup=false
    Mesh Sampling Type=::NewEnumerator5
    Sampled Position Offset=&AddVectorToPosition(
     Position=&SelectVectorFromArray(
      Direct Array Index=$Particles.PartitionParticles.Partition
      Direct Array Mode=::NewEnumerator1
      Vector Selection Array=DataInterface() {propertyValues {FloatData [
         <0.1501 0.0874 -9.3072>
         <-2.5723 2.0119 9.267>
         <8.2457 -3.7918 -1.8664>
         <2.3968 -12.7186 1.0133>
         <-6.3775 -5.168 -8.9636>
         <-0.8792 -12.749 -5.2507>
         <-2.1197 -13.3948 5.2111>
         <0.6948 -7.1689 8.587>
         <-6.2837 -8.1865 8.8226>
         <3.6496 -7.4414 -6.4508>
         <-2.1799 -2.7011 10.6602>
         <-12.4834 4.2506 7.7112>
         <6.3602 -3.498 5.9072>
         <-16.0534 4.9365 -6.5215>
         <5.6867 -8.3193 3.2598>
         <-9.539 -11.9287 4.8964>
         <-8.828 -13.5064 -2.4959>
         <-7.6721 0.6844 9.6566>
         <-7.1311 2.3999 -8.9248>
         <-12.5145 -4.2056 9.0689>
        ]}}
     )
     Vector=<8.6 4.1 0>
    )
    Skeletal Mesh=$System.Skeletal Mesh
    Socket Sampling Mode=::NewEnumerator5
   )
   &ScaleColor(ScaleA=true, ScaleRGB=true)
   /SetVariables(Particles.ChromaFlare:LinearColor=&Multiply_LinearColorByFloat(Linear Color=$Particles.Color, Float=&RandomRangeFloat(Minimum=0.13, Maximum=0.16))) {isSetParameters=true}
   /SetVariables(Particles.FlareColor=&MultiplyLinearColor(
     Linear Color=&HueShiftLinearColor(
      Hue=&RandomRangeFloat(Minimum=-0.02, Maximum=0.02)
      Saturation=&RandomRangeFloat(Minimum=0.8, Maximum=1.2)
      Value=&RandomRangeFloat(Minimum=0.8, Maximum=1)
      Linear Color=$Emitter.FlareColor
     )
     Scale Factor=&MakeVector4FromVectorAndFloat(
      "Vector (RGB)"=<1 1 1>
      "Float (Alpha)"=&Multiply_Float(
       A=&RandomRangeFloat(Minimum=0.3, Maximum=&FN_FrontEndFloatModifier(Gameplay=1, Front End=0.5))
       B=&MakeFloatFromLinearColor(LinearColor=$Particles.Color, Channel:ENiagaraFloatFromLinearColorOptions=::NewEnumerator3)
      )
     )
    )) {isSetParameters=true}
  ]
  rendererClasses [&NiagaraSpriteRendererProperties]
  renderers [
   {rendererClass=&NiagaraSpriteRendererProperties}
   {rendererClass=&NiagaraSpriteRendererProperties, rendererIndex=1}
  ]
 }
 Headwaves {
  emitterUpdateScript [
   &EmitterState(Inactive Response=::NewEnumerator1, Life Cycle Mode=::NewEnumerator1, Loop Duration=0.5, UseLoopDelay=false)
   &SpawnBurst_Instantaneous(
    Spawn Count=2
    Spawn Probability=&MakeCustomFloatFromBool(Bool=&SetBoolByFloatComparison(A=$System.AudioReactivity, B=0.1), True Float=1, False Float=0)
    Use Spawn Probability=true
    UseLoopCountLimit=false
   )
  ]
  particleSpawnScript [
   &PartitionParticles()
   &InitializeParticle(
    Color=&MergeHelper_ColorSwitch(
     0=&MakeCustomLinearColorFromBool(
      Bool=&SetBoolByIntComparison(
       A=&MakeCustomIntFromBool(
        Bool=&Boolean_OrOperation(Bool1=$User.bIsFrontEndPreview, Bool2=$User.bIsFrontEnd)
        True INT=1
        False INT=$Engine.QualityLevel
       )
       B=0
       Comparison Type=::NewEnumerator4
      )
      True Color=<rgba 0.0213 0.0197 0.9844 10>
      False Color=<rgba 0 0.3867 1 20>
     )
     1=&MakeCustomLinearColorFromBool(
      Bool=&SetBoolByIntComparison(
       A=&MakeCustomIntFromBool(
        Bool=&Boolean_OrOperation(Bool1=$User.bIsFrontEndPreview, Bool2=$User.bIsFrontEnd)
        True INT=1
        False INT=$Engine.QualityLevel
       )
       B=0
       Comparison Type=::NewEnumerator4
      )
      True Color=<rgba 0.6354 0.3358 0.2217 10>
      False Color=<rgba 0.6354 0.3358 0.2217 20>
     )
     2=&MakeCustomLinearColorFromBool(
      Bool=&SetBoolByIntComparison(
       A=&MakeCustomIntFromBool(
        Bool=&Boolean_OrOperation(Bool1=$User.bIsFrontEndPreview, Bool2=$User.bIsFrontEnd)
        True INT=1
        False INT=$Engine.QualityLevel
       )
       B=0
       Comparison Type=::NewEnumerator4
      )
      True Color=<rgba 0.1449 0.8463 1 10>
      False Color=<rgba 0.1449 0.8463 1 20>
     )
     3=?
     Index=&MakeIntFromFloat(Float=$User.Variant)
     Num=::NewEnumerator1
    )
    Lifetime=0.5
    Sprite Size Mode=::NewEnumerator3
    Uniform Sprite Size=&SelectFloatFromArray(Direct Array Index=$Particles.PartitionParticles.Partition, Float Selection Array=DataInterface() {propertyValues {FloatData [35,35]}})
    UsePositionOffset=false
    Write Material Random=true
   )
   /SetVariables(Particles.VisibilityTag=&MakeCustomIntFromBool(Bool=$User.bIsFrontEndPreview, True INT=1, False INT=0)) {isSetParameters=true}
  ]
  particleUpdateScript [
   &ParticleState()
   &SkeletalMeshLocation(
    Bone / Socket Index=&SelectIntFromArray(
     Direct Array Index=$Particles.PartitionParticles.Partition
     Direct Array Mode=::NewEnumerator1
     Int Selection Array=DataInterface() {propertyValues {IntData [0,0]}}
    )
    MaskSpawnGroup=false
    Mesh Sampling Type=::NewEnumerator5
    Sampled Position Offset=&SelectVectorFromArray(
     Direct Array Index=$Particles.PartitionParticles.Partition
     Vector Selection Array=DataInterface() {propertyValues {FloatData [<5 10 0>, <7 0 0>]}}
    )
    Skeletal Mesh=$System.Skeletal Mesh
    Socket Sampling Mode=::NewEnumerator5
   )
   &CameraOffset(Camera Offset Amount=&FN_FrontEndFloatModifier(Gameplay=20, Front End=0))
   /SetVariables(Particles.OffsetPosition:NiagaraPosition=&AddVectorToPosition(
     Position=$Particles.Position
     Vector=&RotateVectorByQuaternion(Vector To Rotate=<0 0 0>, Quaternion=$Particles.SkeletalMeshLocation.SampledRotation)
    )) {isSetParameters=true}
   &DynamicMaterialParameters(
    Index 0 Param 1=&MakeFloatFromVector(Vector=Linked() {linkedVariable {name=Particles.Position, type=NiagaraPosition}}, Channel=::NewEnumerator0)
    Index 0 Param 2=&MakeFloatFromVector(Vector=Linked() {linkedVariable {name=Particles.Position, type=NiagaraPosition}}, Channel=::NewEnumerator1)
    Index 0 Param 3=&MakeFloatFromVector(Vector=Linked() {linkedVariable {name=Particles.Position, type=NiagaraPosition}}, Channel=::NewEnumerator2)
    Index 1 Param 1=&MakeFloatFromVector(Vector=Linked() {linkedVariable {name=Particles.OffsetPosition, type=NiagaraPosition}}, Channel=::NewEnumerator0)
    Index 1 Param 2=&MakeFloatFromVector(Vector=Linked() {linkedVariable {name=Particles.OffsetPosition, type=NiagaraPosition}}, Channel=::NewEnumerator1)
    Index 1 Param 3=&MakeFloatFromVector(Vector=Linked() {linkedVariable {name=Particles.OffsetPosition, type=NiagaraPosition}}, Channel=::NewEnumerator2)
    Index 1 Param 4=&Lerp_Float(
     A=&FloatFromCurve(
      FloatCurve=DataInterface() {propertyValues {Curve {keys [
          {arriveTangent=-0.0762, interpMode=RCIM_Cubic, leaveTangent=-0.0762, tangentMode=RCTM_User, time=0, value=0.8}
          {arriveTangent=-0.2085, interpMode=RCIM_Cubic, leaveTangent=-0.2085, tangentMode=RCTM_User, time=1, value=0.3}
         ]}}}
      Scale Curve=1.35
     )
     B=&FloatFromCurve(
      FloatCurve=DataInterface() {propertyValues {Curve {keys [
          {arriveTangent=-0.8006, interpMode=RCIM_Cubic, leaveTangent=-0.8006, tangentMode=RCTM_User, time=0, value=0.8}
          {arriveTangent=-0.3233, interpMode=RCIM_Cubic, leaveTangent=-0.3233, tangentMode=RCTM_User, time=1, value=0.4}
         ]}}}
      Scale Curve=1.35
     )
     Alpha=$System.AudioReactivity
    )
    Module.Index0Float0Write=true
    Module.Index0Float1Write=true
    Module.Index0Float2Write=true
    Module.Index0Float3Write=true
    Module.Index1Float0Write=true
    Module.Index1Float1Write=true
    Module.Index1Float2Write=true
    Module.Index1Float3Write=true
    Write Parameter Index 1=::NewEnumerator1
   )
   &ScaleColor(
    Scale Alpha=&FloatFromCurve(FloatCurve=DataInterface() {propertyValues {Curve {keys [
         {arriveTangent=0.5632, interpMode=RCIM_Cubic, leaveTangent=0.5632, tangentMode=RCTM_User, time=0, value=0}
         {arriveTangent=0.168, interpMode=RCIM_Cubic, leaveTangent=0.168, tangentMode=RCTM_User, time=0.25, value=1}
         {arriveTangent=-0.1012, interpMode=RCIM_Cubic, leaveTangent=-0.1012, tangentMode=RCTM_User, time=1, value=0}
        ]}}})
    Scale RGB=&VectorFromFloat(Value=&FloatFromCurve(FloatCurve=DataInterface() {propertyValues {Curve {keys [
          {arriveTangent=0.168, interpMode=RCIM_Cubic, leaveTangent=0.168, tangentMode=RCTM_User, time=0.2, value=1.5}
          {arriveTangent=-0.1012, interpMode=RCIM_Cubic, leaveTangent=-0.1012, tangentMode=RCTM_User, time=0.4, value=1}
         ]}}}))
    ScaleA=true
    ScaleRGB=true
   )
   &ScaleColor(
    Scale Alpha=$System.ReactiveLerp
    Scale RGB=&VectorFromFloat(Value=&Add_Float(
      A=&Multiply_Float(
       A=&FloatFromCurve(
        FloatCurve=DataInterface() {propertyValues {Curve {keys [
            {interpMode=RCIM_Cubic, tangentMode=RCTM_Break, time=0, value=0}
            {arriveTangent=3.7964, interpMode=RCIM_Cubic, leaveTangent=1.1533, tangentMode=RCTM_Break, time=0.9, value=1}
            <at 1 0>
           ]}}}
        CurveIndex=$System.AudioReactivity
       )
       B=&FN_FrontEndFloatModifier(Gameplay=0.75, Front End=0.5)
      )
      B=$System.ReactiveLerp
     ))
    ScaleA=true
    ScaleRGB=true
   ) {module=ScaleColor003}
   &ScaleColor(
    Color Value To Scale=$Particles.Color
    Scale Alpha=&MakeFloatFromLinearColor(LinearColor=$System.TODColorMultiplier, Channel:ENiagaraFloatFromLinearColorOptions=::NewEnumerator3)
    Scale RGB=&MakeVectorFromLinearColorRGB(Color=$System.TODColorMultiplier)
    ScaleA=true
    ScaleRGB=true
   ) {module=ScaleColor002}
   &ScaleSpriteSize(
    Scale Sprite Size Mode=::NewEnumerator0
    Uniform Scale Factor=&MakeCustomFloatFromBool(Bool=$User.bisHeadHidden, True Float=0, False Float=1)
   )
  ]
  rendererClasses [&NiagaraSpriteRendererProperties]
  renderers [
   {rendererClass=&NiagaraSpriteRendererProperties}
   {rendererClass=&NiagaraSpriteRendererProperties, rendererIndex=1}
   {rendererClass=&NiagaraSpriteRendererProperties, rendererIndex=2}
  ]
 }
}
summary {
 systemName=NS_Character_StrideMiceDeep_Head_Idle
 emitters [
  {emitterName=Headwaves, rendererClasses [&NiagaraSpriteRendererProperties]}
  {emitterName=Flares, rendererClasses [&NiagaraSpriteRendererProperties]}
 ]
 userVariables [
  User.bIsADS {type=NiagaraBool}
  User.bisFootHidden {type=NiagaraBool}
  User.bIsFrontEnd {type=NiagaraBool}
  User.bIsFrontEndPreview {type=NiagaraBool}
  User.bisHeadHidden {type=NiagaraBool}
  User.EnableFrontendFakeAudioLoop {type=NiagaraFloat, defaultValue=0, description=""}
  User.Variant {type=NiagaraFloat, defaultValue=0, description=""}
 ]
}
