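# Post-copy AnimBP for the StrideMiceDeep player body cosmetic, adding secondary motion to zippers, belts, torso and neck.
# Parent mesh pose feeds a parachute-aware Control Rig, zipper bone offset and weighted constraint, then chained spring-based AnimDynamics split by a cached pose.
# A final left-arm zipper rotation reads a DBNO/front-end value, set only in a custom event nothing here calls.
# Update-set sim-space and gravity variables go unread by the AnimGraph.
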AnimBlueprint /BRCosmetics/Characters/Player/Male/Medium/Bodies/M_MED_StrideMiceDeep/Meshes/M_MED_StrideMiceDeep_AnimBP
target_skeleton=&/M_MED_StrideMiceDeep_Skeleton
graphs [
 AnimGraph {nodes [
   n1 = Anim Dynamics - Bone: Dyn_BeltTorso_R!(
    ComponentPose=Local To Component(LocalPose=Use cached pose 'Initial Pose'())
    Alpha=1
    GravityOverride:Vector="0,0,-980"
    ExternalForce="0,0,0"
   ) -> (Pose) {
    angularDampingOverride=0.999
    angularSpringConstant=8000
    bAngularSpring=true
    bOverrideAngularDamping=true
    bUseGravityOverride=true
    gravityScale=0.35
    simulationSpace=RootRelative
    alphaBoolBlend {blendOption=Linear}
    boundBone {boneName=Dyn_BeltTorso_R}
    componentAppliedLinearAccClamp=<3000 3000 3000>
    componentLinearAccScale=<0.5 0.5 0.5>
    componentLinearVelScale=<0.5 0.5 0.5>
    gravityOverride {z=-980}
    physicsBodyDefinitions [{
      boundBone {boneName=Dyn_BeltTorso_R}
      boxExtents=<15 20 20>
      constraintSetup {angularLimitsMax=<0 0 30>, angularLimitsMin=<0 0 -15>, angularTarget=<1 0 0>}
      localJointOffset=<-4 0 0>
     }]
    relativeSpaceBone {boneName=spine_05}
    simSpaceSettings {maxAngularVelocity=100, simSpaceAngularAlpha=0.5, externalAngularVelocity=<1 1 1>}
   }
   n2 = Anim Dynamics - Bone: Dyn_Torso_R!(ComponentPose=@n1, Alpha=1, GravityOverride:Vector="0,0,-980", ExternalForce="0,0,0") -> (Pose) {
    angularDampingOverride=0.999
    angularSpringConstant=10000
    bAngularSpring=true
    bOverrideAngularDamping=true
    bUseGravityOverride=true
    gravityScale=0.35
    simulationSpace=RootRelative
    alphaBoolBlend {blendOption=Linear}
    boundBone {boneName=Dyn_Torso_R}
    componentAppliedLinearAccClamp=<3000 3000 3000>
    componentLinearAccScale=<0.5 0.5 0.5>
    componentLinearVelScale=<0.5 0.5 0.5>
    gravityOverride {z=-980}
    physicsBodyDefinitions [{
      boundBone {boneName=Dyn_Torso_R}
      boxExtents=<15 20 20>
      constraintSetup {angularLimitsMax=<0 25 30>, angularLimitsMin=<0 -25 -30>, angularTarget=<1 0 0>}
      localJointOffset=<-4 0 0>
     }]
    relativeSpaceBone {boneName=spine_05}
    simSpaceSettings {maxAngularVelocity=100, simSpaceAngularAlpha=0.5, externalAngularVelocity=<1 1 1>}
   }
   n3 = Anim Dynamics - Bone: Dyn_Torso_L!(ComponentPose=@n2, Alpha=1, GravityOverride:Vector="0,0,-980", ExternalForce="0,0,0") -> (Pose) {
    angularDampingOverride=0.999
    angularSpringConstant=10000
    bAngularSpring=true
    bOverrideAngularDamping=true
    bUseGravityOverride=true
    gravityScale=0.35
    simulationSpace=RootRelative
    alphaBoolBlend {blendOption=Linear}
    boundBone {boneName=Dyn_Torso_L}
    componentAppliedLinearAccClamp=<3000 3000 3000>
    componentLinearAccScale=<0.5 0.5 0.5>
    componentLinearVelScale=<0.5 0.5 0.5>
    gravityOverride {z=-980}
    physicsBodyDefinitions [{
      boundBone {boneName=Dyn_Torso_L}
      boxExtents=<15 20 20>
      constraintSetup {angularLimitsMax=<0 25 30>, angularLimitsMin=<0 -25 -30>, angularTarget=<1 0 0>}
      localJointOffset=<4 0 0>
     }]
    relativeSpaceBone {boneName=spine_05}
    simSpaceSettings {maxAngularVelocity=100, simSpaceAngularAlpha=0.5, externalAngularVelocity=<1 1 1>}
   }
   n4 = Anim Dynamics - Bone: Dyn_ziipermiddle!(ComponentPose=@n3, Alpha=1, GravityOverride:Vector="0,0,-980", ExternalForce="0,0,0") -> (Pose) {
    angularDampingOverride=0.999
    angularSpringConstant=1000
    bAngularSpring=true
    bOverrideAngularDamping=true
    bUseGravityOverride=true
    gravityScale=0.35
    simulationSpace=RootRelative
    alphaBoolBlend {blendOption=Linear}
    boundBone {boneName=Dyn_ziipermiddle}
    componentAppliedLinearAccClamp=<3000 3000 3000>
    componentLinearAccScale=<0.5 0.5 0.5>
    componentLinearVelScale=<0.5 0.5 0.5>
    gravityOverride {z=-980}
    physicsBodyDefinitions [{
      boundBone {boneName=Dyn_ziipermiddle}
      boxExtents=<15 20 20>
      constraintSetup {angularLimitsMax=<0 20 35>, angularLimitsMin=<0 -20 0>, angularTarget=<1 0 0>}
      localJointOffset=<5 0 0>
     }]
    relativeSpaceBone {boneName=spine_05}
    simSpaceSettings {maxAngularVelocity=100, simSpaceAngularAlpha=0.5, externalAngularVelocity=<1 1 1>}
   }
   n5 = Anim Dynamics - Bone: Dyn_neck!(ComponentPose=@n4, Alpha=1, GravityOverride:Vector="0,0,-980", ExternalForce="0,0,0") -> (Pose) {
    angularDampingOverride=0.999
    angularSpringConstant=4000
    bAngularSpring=true
    bOverrideAngularDamping=true
    bUseGravityOverride=true
    gravityScale=0.35
    simulationSpace=RootRelative
    alphaBoolBlend {blendOption=Linear}
    boundBone {boneName=Dyn_neck}
    componentAppliedLinearAccClamp=<3000 3000 3000>
    componentLinearAccScale=<0.5 0.5 0.5>
    componentLinearVelScale=<0.5 0.5 0.5>
    gravityOverride {z=-980}
    physicsBodyDefinitions [{
      boundBone {boneName=Dyn_neck}
      boxExtents=<15 25 30>
      constraintSetup {angularLimitsMax=<0 0 25>, angularLimitsMin=<0 -50 -20>, angularTarget=<1 0 0>}
      localJointOffset=<5 0 0>
     }]
    relativeSpaceBone {boneName=spine_05}
    simSpaceSettings {maxAngularVelocity=100, simSpaceAngularAlpha=0.5, externalAngularVelocity=<1 1 1>}
   }
   n6 = Anim Dynamics - Bone: Dyn_ZiiperArm_L!(ComponentPose=@n5, Alpha=1, GravityOverride:Vector="0,0,-980", ExternalForce="0,0,0") -> (Pose) {
    angularDampingOverride=0.999
    angularSpringConstant=1000
    bAngularSpring=true
    bOverrideAngularDamping=true
    bUseGravityOverride=true
    gravityScale=0.35
    simulationSpace=RootRelative
    alphaBoolBlend {blendOption=Linear}
    boundBone {boneName=Dyn_ZiiperArm_L}
    componentAppliedLinearAccClamp=<10000 10000 10000>
    componentLinearAccScale=<1 1 1>
    componentLinearVelScale=<1 1 1>
    gravityOverride {z=-980}
    physicsBodyDefinitions [{
      boundBone {boneName=Dyn_ZiiperArm_L}
      boxExtents=<20 20 20>
      constraintSetup {angularLimitsMax=<0 100 0>, angularLimitsMin=<0 -5 0>, angularTarget=<1 0 0>}
      localJointOffset=<4 0 0>
     }]
    relativeSpaceBone {boneName=spine_05}
    simSpaceSettings {maxAngularVelocity=100, simSpaceAngularAlpha=0.5, externalAngularVelocity=<1 1 1>}
   }
   n7 = Anim Dynamics - Bone: Dyn_ZiiperWrist_L!(ComponentPose=@n6, Alpha=1, GravityOverride:Vector="0,0,-980", ExternalForce="0,0,0") -> (Pose) {
    angularDampingOverride=0.999
    angularSpringConstant=10000
    bAngularSpring=true
    bOverrideAngularDamping=true
    bUseGravityOverride=true
    gravityScale=0.35
    simulationSpace=RootRelative
    alphaBoolBlend {blendOption=Linear}
    boundBone {boneName=Dyn_ZiiperWrist_L}
    componentAppliedLinearAccClamp=<3000 3000 3000>
    componentLinearAccScale=<0.5 0.5 0.5>
    componentLinearVelScale=<0.5 0.5 0.5>
    gravityOverride {z=-980}
    physicsBodyDefinitions [{
      boundBone {boneName=Dyn_ZiiperWrist_L}
      boxExtents=<15 20 20>
      constraintSetup {angularLimitsMax=<15 15 90>, angularLimitsMin=<-15 -15 -5>, angularTarget=<1 0 0>}
      localJointOffset=<4 0 0>
     }]
    relativeSpaceBone {boneName=spine_05}
    simSpaceSettings {maxAngularVelocity=100, simSpaceAngularAlpha=0.5, externalAngularVelocity=<1 1 1>}
   }
   n8 = Anim Dynamics - Bone: Dyn_ZiiperWrist_R!(ComponentPose=@n7, Alpha=1, GravityOverride:Vector="0,0,-980", ExternalForce="0,0,0") -> (Pose) {
    angularDampingOverride=0.999
    angularSpringConstant=10000
    bAngularSpring=true
    bOverrideAngularDamping=true
    bUseGravityOverride=true
    gravityScale=0.35
    simulationSpace=RootRelative
    alphaBoolBlend {blendOption=Linear}
    boundBone {boneName=Dyn_ZiiperWrist_R}
    componentAppliedLinearAccClamp=<3000 3000 3000>
    componentLinearAccScale=<0.5 0.5 0.5>
    componentLinearVelScale=<0.5 0.5 0.5>
    gravityOverride {z=-980}
    physicsBodyDefinitions [{
      boundBone {boneName=Dyn_ZiiperWrist_R}
      boxExtents=<15 20 20>
      constraintSetup {angularLimitsMax=<15 15 90>, angularLimitsMin=<-15 -15 -5>, angularTarget=<1 0 0>}
      localJointOffset=<4 0 0>
     }]
    relativeSpaceBone {boneName=spine_05}
    simSpaceSettings {maxAngularVelocity=100, simSpaceAngularAlpha=0.5, externalAngularVelocity=<1 1 1>}
   }
   n9 = "Transform (Modify) Bone - Bone: Base_ZiiperArm_L"(ComponentPose=@n8, Alpha=1, Translation="0,0,0", Rotation=Get ArmZipperDBNO() -> (ArmZipperDBNO:Rotator), Scale="1,1,1") {rotationMode=BMM_Additive, rotationSpace=BCS_BoneSpace, alphaBoolBlend {blendOption=Linear}, boneToModify {boneName=Base_ZiiperArm_L}}
   Output Pose(Result=Component To Local(ComponentPose=@n9)) {layerGroup=None}
   n10 = Control Rig!(Source=Copy Pose From Mesh() {bUseAttachedParent=true}, Alpha, Is_Parachuting:double=Get Is_Parachute_Open() -> (Is_Parachute_Open:double)) -> (Pose) {
    alphaBoolBlend {blendOption=Linear}
    controlRigAssetReference {blueprintRigClass=&/M_MED_StrideMiceDeep_CtrlRig.M_MED_StrideMiceDeep_CtrlRig_C}
   }
   n11 = "Transform (Modify) Bone - Bone: Base_ZiiperArm_L"!(ComponentPose=Local To Component(LocalPose=@n10), bAlphaBoolEnabled:bool=True, Translation="0.35,0,0", Rotation="0,0,0", Scale="1,1,1") -> (Pose) {
    alphaInputType=Bool
    rotationMode=BMM_Additive
    translationMode=BMM_Additive
    alphaBoolBlend {blendOption=Linear}
    boneToModify {boneName=Base_ZiiperArm_L}
    translation {x=0.35}
   }
   n12 = "Constraint - Base_ZipperPants_L (PPPPPPPP)"!(
    ComponentPose=@n11
    Alpha=1
    ConstraintWeights_0=0.4023
    ConstraintWeights_1:float=0.2978
    ConstraintWeights_2:float=0.2763
    ConstraintWeights_3:float=0.0101
    ConstraintWeights_4:float=0.0073
    ConstraintWeights_5:float=0.0027
    ConstraintWeights_6:float=0.0026
    ConstraintWeights_7:float=0.001
   ) -> (Pose) {
    alphaBoolBlend {blendOption=Linear}
    boneToModify {boneName=Base_ZipperPants_L}
    constraintSetup [
     {targetBone {boneName=thigh_twist_01_l}}
     {targetBone {boneName=pelvis}}
     {targetBone {boneName=deform_groin_l}}
     {targetBone {boneName=spine_01}}
     {targetBone {boneName=thigh_l}}
     {targetBone {boneName=thigh_twist_01_r}}
     {targetBone {boneName=deform_glute_l}}
     {targetBone {boneName=deform_groin_r}}
    ]
    constraintWeights [1,1,1,1,1,1,1,1]
   }
   n13 = Anim Dynamics - Bone: Dyn_ZipperAnkle_R!(ComponentPose=@n12, Alpha=1, GravityOverride:Vector="0,0,-980", ExternalForce="0,0,0") -> (Pose) {
    angularDampingOverride=0.999
    angularSpringConstant=4000
    bAngularSpring=true
    bOverrideAngularDamping=true
    bUseGravityOverride=true
    gravityScale=0.35
    simulationSpace=RootRelative
    alphaBoolBlend {blendOption=Linear}
    boundBone {boneName=Dyn_ZipperAnkle_R}
    componentAppliedLinearAccClamp=<3000 3000 3000>
    componentLinearAccScale=<0.5 0.5 0.5>
    componentLinearVelScale=<0.5 0.5 0.5>
    gravityOverride {z=-980}
    physicsBodyDefinitions [{
      boundBone {boneName=Dyn_ZipperAnkle_R}
      boxExtents=<15 20 20>
      constraintSetup {angularLimitsMax=<0 10 2.5>, angularLimitsMin=<0 -60 -2.5>, angularTarget=<1 0 0>}
      localJointOffset=<-3 0 0>
     }]
    relativeSpaceBone {boneName=spine_05}
    simSpaceSettings {maxAngularVelocity=100, simSpaceAngularAlpha=0.5, externalAngularVelocity=<1 1 1>}
   }
   n14 = Anim Dynamics - Bone: Dyn_ZipperAnkle_L!(ComponentPose=@n13, Alpha=1, GravityOverride:Vector="0,0,-980", ExternalForce="0,0,0") -> (Pose) {
    angularDampingOverride=0.999
    angularSpringConstant=10000
    bAngularSpring=true
    bOverrideAngularDamping=true
    bUseGravityOverride=true
    gravityScale=0.35
    simulationSpace=RootRelative
    alphaBoolBlend {blendOption=Linear}
    boundBone {boneName=Dyn_ZipperAnkle_L}
    componentAppliedLinearAccClamp=<3000 3000 3000>
    componentLinearAccScale=<0.5 0.5 0.5>
    componentLinearVelScale=<0.5 0.5 0.5>
    gravityOverride {z=-980}
    physicsBodyDefinitions [{
      boundBone {boneName=Dyn_ZipperAnkle_L}
      boxExtents=<15 20 20>
      constraintSetup {angularLimitsMax=<0 60 2.5>, angularLimitsMin=<0 -10 -2.5>, angularTarget=<1 0 0>}
      localJointOffset=<3 0 0>
     }]
    relativeSpaceBone {boneName=spine_05}
    simSpaceSettings {maxAngularVelocity=100, simSpaceAngularAlpha=0.5, externalAngularVelocity=<1 1 1>}
   }
   n15 = Anim Dynamics - Bone: Dyn_ZipperPants_L!(ComponentPose=@n14, Alpha=1, GravityOverride:Vector="0,0,-980", ExternalForce="0,0,0") -> (Pose) {
    angularDampingOverride=0.999
    angularSpringConstant=10000
    bAngularSpring=true
    bOverrideAngularDamping=true
    bUseGravityOverride=true
    gravityScale=0.35
    simulationSpace=RootRelative
    alphaBoolBlend {blendOption=Linear}
    boundBone {boneName=Dyn_ZipperPants_L}
    componentAppliedLinearAccClamp=<3000 3000 3000>
    componentLinearAccScale=<0.5 0.5 0.5>
    componentLinearVelScale=<0.5 0.5 0.5>
    gravityOverride {z=-980}
    physicsBodyDefinitions [{
      boundBone {boneName=Dyn_ZipperPants_L}
      boxExtents=<15 20 20>
      constraintSetup {angularLimitsMax=<0 30 30>, angularLimitsMin=<0 -30 0>, angularTarget=<1 0 0>}
      localJointOffset=<5 0 0>
     }]
    relativeSpaceBone {boneName=spine_05}
    simSpaceSettings {maxAngularVelocity=100, simSpaceAngularAlpha=0.5, externalAngularVelocity=<1 1 1>}
   }
   n16 = Anim Dynamics - Bone: Dyn_ZipperPants_R!(ComponentPose=@n15, Alpha=1, GravityOverride:Vector="0,0,-980", ExternalForce="0,0,0") -> (Pose) {
    angularDampingOverride=0.999
    angularSpringConstant=10000
    bAngularSpring=true
    bOverrideAngularDamping=true
    bUseGravityOverride=true
    gravityScale=0.35
    simulationSpace=RootRelative
    alphaBoolBlend {blendOption=Linear}
    boundBone {boneName=Dyn_ZipperPants_R}
    componentAppliedLinearAccClamp=<3000 3000 3000>
    componentLinearAccScale=<0.5 0.5 0.5>
    componentLinearVelScale=<0.5 0.5 0.5>
    gravityOverride {z=-980}
    physicsBodyDefinitions [{
      boundBone {boneName=Dyn_ZipperPants_R}
      boxExtents=<15 20 20>
      constraintSetup {angularLimitsMax=<0 40 30>, angularLimitsMin=<0 -30 -5>, angularTarget=<1 0 0>}
      localJointOffset=<-5 0 0>
     }]
    relativeSpaceBone {boneName=spine_05}
    simSpaceSettings {maxAngularVelocity=100, simSpaceAngularAlpha=0.5, externalAngularVelocity=<1 1 1>}
   }
   n17 = Anim Dynamics - Bone: Dyn_BeltTorso_L!(ComponentPose=@n16, Alpha=1, GravityOverride:Vector="0,0,-980", ExternalForce="0,0,0") -> (Pose) {
    angularDampingOverride=0.999
    angularSpringConstant=10000
    bAngularSpring=true
    bOverrideAngularDamping=true
    bUseGravityOverride=true
    gravityScale=0.35
    simulationSpace=RootRelative
    alphaBoolBlend {blendOption=Linear}
    boundBone {boneName=Dyn_BeltTorso_L}
    componentAppliedLinearAccClamp=<3000 3000 3000>
    componentLinearAccScale=<0.5 0.5 0.5>
    componentLinearVelScale=<0.5 0.5 0.5>
    gravityOverride {z=-980}
    physicsBodyDefinitions [{
      boundBone {boneName=Dyn_BeltTorso_L}
      boxExtents=<15 20 20>
      constraintSetup {angularLimitsMax=<0 0 15>, angularLimitsMin=<0 0 -30>, angularTarget=<1 0 0>}
      localJointOffset=<4 0 0>
     }]
    relativeSpaceBone {boneName=spine_05}
    simSpaceSettings {maxAngularVelocity=100, simSpaceAngularAlpha=0.5, externalAngularVelocity=<1 1 1>}
   }
   Save cached pose 'Initial Pose'(Pose=Component To Local(ComponentPose=@n17)) {cacheName=Initial Pose}
  ]}
 EventGraph {nodes [
   K2Node_Event_0 = CustomEvent!() -> (OutputDelegate, then, DeltaTimeX:float=~) {is=title K2Node_CustomEvent}
   > n1 = Sequence()
   > Branch(Condition=Get bIsTargeting())
   > Set ArmZipperCompensate(ArmZipperCompensate:Rotator=MakeRotator(Yaw=-4)) -> (Output_Get:Rotator)
   n2 = Branch(execute=@n1.then_1, Condition=OR Boolean(A=Get bIsDBNO(), B=Get bIsInFrontEnd()))
   > Set ArmZipperDBNO(ArmZipperDBNO:Rotator="0,0,-18") -> (Output_Get:Rotator)
   Set ArmZipperDBNO(execute=@n2.else, ArmZipperDBNO="0, 0, 0")
   Event BlueprintUpdateAnimation()
   > n3 = Sequence()
   > n4 = Branch(Condition=Get bIsInFrontEnd())
   > Set Sim Space Settings(Sim Space Settings:Sim Space Settings Structure=Get Sim Space Settings Front End() -> (Sim Space Settings Front End:Sim Space Settings Structure)) -> (Output_Get:Sim Space Settings Structure)
   > Set Override World Gravity PIE(Override World Gravity PIE:Vector="0,0,-1660") -> (Output_Get:Vector)
   Set Sim Space Settings(execute=@n4.else, Sim Space Settings=Get Sim Space Settings PIE() -> (Sim Space Settings PIE:Sim Space Settings Structure))
   > Set Override World Gravity PIE(Override World Gravity PIE="0,0,-980")
   Set Is_Parachute_Open(execute=@n3.then_1, Is_Parachute_Open:double="To Float (Boolean)"(InBool=Get bIsParachuteOpened())) -> (Output_Get:double)
   Set Sky Diving(Sky Diving:bool=Get bIsSkyDiving()) -> (Output_Get:bool)
   Set Sky Floating(Sky Floating:bool=Get bIsSkydiveFloating()) -> (Output_Get:bool)
  ]}
]
variables [
 ArmZipperCompensate {type=Rotator}
 ArmZipperDBNO {type=Rotator}
 Is_Parachute_Open {type=number}
 Override World Gravity PIE {type=Vector, default=<0 0 -1660>}
 Override World Gravity_0 {type=Vector, default=<0 0 -980>}
 Sim Space Settings {
  type=SimSpaceSettings
  default {
   dampingAlpha=1
   maxAngularAcceleration=10000
   maxAngularVelocity=10000
   maxLinearAcceleration=10000
   maxLinearVelocity=10000
   velocityScaleZ=1
   worldAlpha=0
   externalLinearDragV=<0 0 0>
  }
 }
 Sim Space Settings Front End {
  type=SimSpaceSettings
  default {
   dampingAlpha=1
   maxAngularAcceleration=10000
   maxAngularVelocity=10000
   maxLinearAcceleration=10000
   maxLinearVelocity=10000
   velocityScaleZ=1
   worldAlpha=0
   externalLinearDragV=<0 0 0>
  }
 }
 Sim Space Settings PIE {
  type=SimSpaceSettings
  default {
   dampingAlpha=1
   maxAngularAcceleration=10000
   maxAngularVelocity=10000
   maxLinearAcceleration=10000
   maxLinearVelocity=10000
   velocityScaleZ=1
   worldAlpha=0
   externalLinearDragV=<0 0 0>
  }
 }
 Sky Diving {type=boolean}
 Sky Floating {type=boolean}
]
